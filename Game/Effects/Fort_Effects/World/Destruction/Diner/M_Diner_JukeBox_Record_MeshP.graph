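# Default Lit mesh-particle material for a jukebox record, part of the diner destruction effects.
# It draws the record procedurally from UV-centered sphere masks, with no textures: a larger center disc scaled by ColorCenter, minus a small inner spot.
# The inverse of the scaled disc multiplies the Metallic and Specular parameters, and the disc also drives a constant blend.
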Material /Game/Effects/Fort_Effects/World/Destruction/Diner/M_Diner_JukeBox_Record_MeshP
graph {
 properties {
  bUsedWithMeshParticles=true
  pixelDepthOffsetMode=PDOM_Legacy
  previewMesh=/Game/Packages/DS_Fortnight/SM/Mesh/S_Siphon_GlassTubeContents
  usedShadingModels=Default Lit
 }
 nodes [
  n1 = Multiply(
   A=SphereMask(
    A=TextureCoordinate()
    B=Constant2Vector() {g=0.5, r=0.5}
    Radius=ScalarParameter() {defaultValue=0.25, parameterName=Radius}
    Hardness=ScalarParameter() {defaultValue=0.8, parameterName=Hardness}
   )
   B=ScalarParameter() {defaultValue=0.15, parameterName=ColorCenter}
  )
  n2 = Power(Base=SphereMask(
    A=TextureCoordinate()
    B=Constant2Vector() {g=0.5, r=0.5}
    Radius=ScalarParameter() {defaultValue=0.05, parameterName=Radius_1}
    Hardness=ScalarParameter() {parameterName=Hardness_1}
   ))
  n3 = Add(A=@n2, B=@n2)
  Add(A=Constant(), B=Multiply(A=@n1, B=OneMinus(None=@n3)))
  LinearInterpolate(A=Constant() {r=1}, B=Constant(), Alpha=@n1)
  n4 = OneMinus(None=@n1)
  Multiply(A=ScalarParameter() {defaultValue=0.5, parameterName=Metallic}, B=@n4)
  Multiply(A=ScalarParameter() {defaultValue=1, parameterName=Specular}, B=@n4)
  ParticleColor()
  Power()
  Power(Base=Multiply(A=@n4, B=@n4))
  ScalarParameter() {defaultValue=0.15, parameterName=ColorCenter_1}
  SphereMask()
 ]
}
